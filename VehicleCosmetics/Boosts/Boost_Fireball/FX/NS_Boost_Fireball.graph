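# Niagara system for the Fireball vehicle boost cosmetic, layering flipbook fire, trail, ribbon, spinning mesh sphere, glow and stardust emitters.
# Emitters choose colors through painted-variant overrides, with User parameters driving offsets, scale, painted and frontend-preview tuning.
# Merged emitters use spawn groups, switch helpers and visibility tags to run two particle types each.
# Fire and ribbon particles inherit source movement, fading over life; skeletal-socket placement and some duplicate emitters are disabled.
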NiagaraSystem /VehicleCosmetics/Boosts/Boost_Fireball/FX/NS_Boost_Fireball
dependencies {
 usedDataInterfaces [
  &NiagaraDataInterfaceCurve
  &NiagaraDataInterfaceSpriteRendererInfo
  &NiagaraDataInterfaceSkeletalMesh
 ]
 usedDynamicInputs []
 usedModules [
  &self:SystemSpawnScript.SystemScriptSource.SystemScriptGraph.NiagaraNodeAssignment_0.SetVariables_#1
  &SystemState
 ]
}
emitters {
 NE_VC_Boost_CoreBigFire {
  emitterUpdateScript [
   &EmitterState(Life Cycle Mode=::NewEnumerator1, Loop Duration=0.01, UseLoopDelay=false)
   &SpawnBurst_Instantaneous(Spawn Count=20, Use Spawn Probability=false, UseLoopCountLimit=true)
   &VC_FilteredSocketCount() {enabled=false}
   &SpawnRate(
    SpawnRate=&MakeCustomFloatFromBool(Bool=$User.bIsFrontendPreview, True Float=120, False Float=0)
    Use Spawn Probability=false
   )
  ]
  particleSpawnScript [
   &VC_PaintedVariantsOverride(
    Black=<rgba 0.001 0.001 0.001 0.9>
    BurntSienna=<rgba 0.07 0.0196 0.014 1>
    Cobalt=<rgba 0.03 0.0395 0.6 1>
    Crimson=<rgba 0.2 0.0011 0 1>
    DefaultColor=<rgba 4 0.6667 0 0.05>
    ForestGreen=<rgba 0.0625 0.15 0 1.2>
    Grey=<rgba 0.02 0.02 0.02 1>
    Lime=<rgba 0.1 0.15 0 1>
    Orange=<rgba 0.5 0.125 0 0.5>
    Pink=<rgba 0.2 0 0.1273 1>
    Purple=<rgba 0.05 0 0.15 1.3>
    Saffron=<rgba 0.2 0.2 0 0.5>
    SkyBlue=<rgba 0 0.175 0.3 0.5>
    TitaniumWhite=<rgba 0.5 0.5 0.5 0.2>
   )
   &InitializeParticle(
    Color=$Particles.Color
    Lifetime Min:NiagaraFloat=&MakeCustomFloatFromBool(Bool=$User.bIsPainted, True Float=1, False Float=0.5)
    Lifetime Mode=::NewEnumerator1
    Mass Mode=::NewEnumerator3
    Position Offset=&AddVectorToPosition(
     Position=&AddVectorToPosition(Position=<-55 0 0>, Vector=$User.SingleBoostPos)
     Vector=&Multiply_VectorByFloat(Vector=$User.DTCarOffset, Float=1)
    )
    Position Offset Coordinate Space=::Local
    Sprite Rotation Mode=::NewEnumerator2
    Sprite Size Mode=::NewEnumerator1
    Uniform Sprite Size Max=&Multiply_Float(A=54, B=$User.Scale)
    Uniform Sprite Size Min=&Multiply_Float(A=44, B=$User.Scale)
    UsePositionOffset=true
    Write Material Random=true
   )
   &PartitionParticles(Number Of Partitions=$Emitter.PortCount) {enabled=false}
   &SkeletalMeshLocation(
    Bone / Socket Index=$Particles.PartitionParticles.Partition
    Bone / Socket Sampling Mode=::NewEnumerator5
    MaskSpawnGroup=false
    Mesh Sampling Type=::NewEnumerator6
    Sampled Position Offset=&AddVectorToPosition(
     Position=&AddVectorToPosition(Position=<-20 0 0>, Vector=$User.SingleBoostPos)
     Vector=$User.DTCarOffset
    )
    Skeletal Mesh=$System.SkeletalMesh
   ) {enabled=false}
   &AddVelocity(Velocity=<2000 0 0>, Velocity Speed Scale=-1)
  ]
  particleUpdateScript [
   &ParticleState()
   &ScaleSpriteSize(Uniform Curve Sprite Scale=DataInterface() {propertyValues {Curve {keys [<at 0 1.2 RCIM_Cubic>, {arriveTangent=-4.8, interpMode=RCIM_Cubic, leaveTangent=-4.8, time=0.1, value=1}, <at 0.25 0 RCIM_Cubic>]}}}) {module=ScaleSpriteSize002}
   &Drag(Drag=0.5, Use Linear Drag=true, Use Rotational Drag=true)
   &InheritSourceMovement(
    Applied Position Scale=&Multiply_VectorByFloat(
     Vector=<1 1 1>
     Float=&FloatFromCurve(FloatCurve=DataInterface() {propertyValues {Curve {keys [{arriveTangent=-2.8649, interpMode=RCIM_Cubic, leaveTangent=-2.8649, tangentMode=RCTM_User, time=0, value=1}, <at 1 0 RCIM_Cubic>]}}})
    )
    ApplyForce=false
    ApplyPosition=true
    ApplyPositionFalloff=false
   )
   &SolveForcesAndVelocity(Clamp Velocity=false, Limit Acceleration=false)
   &SubUVAnimation(
    End Frame Range Override=55
    "Sprite Renderer "=DataInterface() {propertyValues {SpriteRenderer=&self:NE_VC_Boost_CoreBigFire_0.NiagaraSpriteRendererProperties_0}}
    Start Frame Range Override=15
    SubUV Animation Mode:ENiagara_SubUVLookupModeV2=::NewEnumerator1
    SubUV Curve=DataInterface() {propertyValues {Curve {keys [
        {arriveTangent=0.178, interpMode=RCIM_Cubic, leaveTangent=0.178, tangentMode=RCTM_User, time=0, value=0}
        {arriveTangent=-0.0346, interpMode=RCIM_Cubic, leaveTangent=-0.0346, tangentMode=RCTM_User, time=0.5, value=1}
       ]}}}
    UseEndFrame=false
    UseStartFrame=true
   )
   &ScaleColor(
    Scale Alpha=&FloatFromCurve(
     FloatCurve=DataInterface() {propertyValues {Curve {keys [<at 0 0.75>,<at 0.05 0.85>,<at 1 0>]}}}
     Scale Curve=&MakeCustomFloatFromBool(Bool=$User.bIsPainted, True Float=0.7, False Float=1)
    )
    Scale RGB=&VectorFromFloat(Value=&FloatFromCurve(
      FloatCurve=DataInterface() {propertyValues {Curve {keys [<at 0 1>,<at 1 0>]}}}
      Scale Curve=&MakeCustomFloatFromBool(Bool=$User.bIsPainted, True Float=2.5, False Float=2)
     ))
    ScaleA=true
    ScaleRGB=true
   )
  ]
  rendererClasses [&NiagaraSpriteRendererProperties]
  renderers [{rendererClass=&NiagaraSpriteRendererProperties}]
 }
 NE_VC_Boost_FireTrail {
  emitterUpdateScript [
   &EmitterState(Life Cycle Mode=::NewEnumerator1, Loop Duration=1, UseLoopDelay=false)
   &VC_FilteredSocketCount() {enabled=false}
   &SpawnRate(
    SpawnRate=&MakeCustomFloatFromBool(
     Bool=$User.bIsFrontendPreview
     True Float=200
     False Float=&MakeCustomFloatFromBool(Bool=$User.bIsPainted, True Float=300, False Float=250)
    )
    Use Spawn Probability=false
   )
   &SpawnRate(
    SpawnGroup=1
    SpawnRate=&MakeCustomFloatFromBool(Bool=$User.bIsFrontendPreview, True Float=120, False Float=250)
    Use Spawn Probability=false
   ) {module=SpawnRate001}
  ]
  particleSpawnScript [
   &VC_PaintedVariantsOverride(
    Black=<rgba 0.001 0.001 0.001 5>
    BurntSienna=<rgba 0.07 0.0196 0.014 1>
    Cobalt=<rgba 0.0927 0.1066 0.9271 1>
    Crimson=<rgba 0.1 0.0016 0 2>
    DefaultColor=<rgba 4 0.3333 0 1>
    ForestGreen=<rgba 0.0167 0.05 0 1>
    Grey=<rgba 0.02 0.02 0.02 1>
    Lime=<rgba 0.1 0.15 0 1>
    Orange=<rgba 1 0.25 0 1>
    Pink=<rgba 0.2 0.0125 0.1318 1>
    Purple=<rgba 0.0417 0 0.125 1>
    R1 Black:LinearColor=<rgba 0.04 0.04 0.04 5>
    R1 BurntSienna:LinearColor=<rgba 0.4 0.1097 0.0833 1>
    R1 Cobalt:LinearColor=<rgba 0.4 0.46 4 1>
    R1 Crimson:LinearColor=<rgba 2 0.0322 0 2>
    R1 DefaultColor=<rgba 4 0.3333 0 1>
    R1 ForestGreen:LinearColor=<rgba 0 0.5729 0 2>
    R1 Grey:LinearColor=<rgba 0.2 0.2 0.2 3>
    R1 Lime:LinearColor=<rgba 0.6667 1 0 2>
    R1 Pink:LinearColor=<rgba 2 0.125 1.3182 1>
    R1 Purple:LinearColor=<rgba 0.5 0 1.5 1>
    R1 Saffron:LinearColor=<rgba 2 1.5333 0 1>
    R1 SkyBlue:LinearColor=<rgba 0 1.9962 3 1>
    R1 TitaniumWhite:LinearColor=<rgba 1 1 1 1>
    Saffron=<rgba 0.1 0.0767 0 4>
    SkyBlue=<rgba 0 0.2994 0.45 1>
    TitaniumWhite=<rgba 1 1 1 0.3>
    Use R1 Painted Overrides=true
   )
   &InitializeParticle(
    Color=&MergeHelper_ColorSwitch(0=$Particles.Color, 1=$Particles.Color, 2=?, 3=?, Index=$Emitter.SpawnGroup, Num=::NewEnumerator0)
    Lifetime=&MergeHelper_FloatSwitch(
     0=1
     1=&RandomRangeFloat(Minimum=0.3, Maximum=0.4)
     2=?
     3=?
     Index=$Emitter.SpawnGroup
     Num=::NewEnumerator0
    )
    Mass Mode=::NewEnumerator3
    Position Offset=&MergeHelper_VectorSwitch(
     0=&AddVectorToPosition(
      Position=&AddVectorToPosition(Position=<-50 0 0>, Vector=$User.SingleBoostPos)
      Vector=&Multiply_VectorByFloat(Vector=$User.DTCarOffset, Float=1)
     )
     1=&AddVectorToPosition(
      Position=&AddVectorToPosition(Position=<-75 0 0>, Vector=$User.SingleBoostPos)
      Vector=&Multiply_VectorByFloat(Vector=$User.DTCarOffset, Float=1)
     )
     2=?
     3=?
     Index=$Emitter.SpawnGroup
     Num=::NewEnumerator0
    )
    Position Offset Coordinate Space=::Local
    Sprite Rotation Mode=::NewEnumerator2
    Sprite Size Mode=::NewEnumerator3
    Uniform Sprite Size=&MergeHelper_FloatSwitch(
     0=&RandomRangeFloat(Minimum=60, Maximum=110)
     1=&RandomRangeFloat(Minimum=3.5, Maximum=4)
     2=?
     3=?
     Index=$Emitter.SpawnGroup
     Num=::NewEnumerator0
    )
    UsePositionOffset=true
    Write Material Random=true
   )
   &PartitionParticles(Number Of Partitions=$Emitter.PortCount) {enabled=false}
   &SkeletalMeshLocation(
    Bone / Socket Index=$Particles.PartitionParticles.Partition
    Bone / Socket Sampling Mode=::NewEnumerator5
    MaskSpawnGroup=false
    Mesh Sampling Type=::NewEnumerator6
    Sampled Position Offset=&AddVectorToPosition(Position=<-10 0 0>, Vector=$User.DTCarOffset)
    Skeletal Mesh=$System.SkeletalMesh
   ) {enabled=false}
   &AddVelocity(
    Velocity=&MergeHelper_VectorSwitch(0=<2000 0 0>, 1=<1000 0 0>, 2=?, 3=?, Index=$Emitter.SpawnGroup, Num=::NewEnumerator0)
    Velocity Speed Scale=&MergeHelper_FloatSwitch(
     0=-1
     1=&Multiply_Float(
      A=&MakeCustomFloatFromBool(Bool=$User.bIsFrontendPreview, True Float=-1, False Float=-0.7)
      B=$User.Scale
     )
     2=?
     3=?
     Index=$Emitter.SpawnGroup
     Num=::NewEnumerator0
    )
   )
   &ShapeLocation(
    MaskSpawnGroup=false
    Override Seed=false
    Sphere Radius=&MergeHelper_FloatSwitch(0=0, 1=50, 2=?, 3=?, Index=$Emitter.SpawnGroup, Num=::NewEnumerator0)
    UseOwnerScale=true
   ) {module=ShapeLocation001}
   &ShapeLocation(
    MaskSpawnGroup=false
    Override Seed=false
    Sphere Radius=&RandomRangeFloat(Minimum=0, Maximum=25)
    UseOwnerScale=true
   ) {enabled=false}
   /SetVariables(Particles.VisibilityTag=$Emitter.SpawnGroup) {isSetParameters=true}
  ]
  particleUpdateScript [
   &ParticleState()
   &SpriteRotationRate(
    Delta Time:NiagaraFloat=$Particles.Lifetime
    Rotation Rate=&MergeHelper_FloatSwitch(
     0=&RandomRangeFloat(Minimum=10, Maximum=-10)
     1=0
     2=?
     3=?
     Index=&SpawnGroup()
     Num=::NewEnumerator0
    )
   )
   &ScaleSpriteSize(
    Scale Sprite Size Mode=::NewEnumerator0
    Uniform Scale Factor=&MergeHelper_FloatSwitch(
     0=&FloatFromCurve(
      FloatCurve=DataInterface() {propertyValues {Curve {keys [
          <at 0 0.5 RCIM_Cubic>
          {arriveTangent=-0.3122, interpMode=RCIM_Cubic, leaveTangent=-0.3122, tangentMode=RCTM_User, time=0.02, value=0.8}
          {arriveTangent=-0.9415, interpMode=RCIM_Cubic, leaveTangent=-0.9415, tangentMode=RCTM_User, time=0.5, value=0}
         ]}}}
      Scale Curve=1.5
     )
     1=&FloatFromCurve(FloatCurve=DataInterface() {propertyValues {Curve {keys [<at 0.2 1 RCIM_Cubic>,<at 1 0.5 RCIM_Cubic>]}}}, Scale Curve=2)
     2=?
     3=?
     Index=&SpawnGroup()
     Num=::NewEnumerator0
    )
   )
   &WindForce(
    Frequency=15
    Scale=50
    Wind Speed=<2 0 2>
    Wind Speed Scale=&MergeHelper_FloatSwitch(0=100, 1=0, 2=?, 3=?, Index=&SpawnGroup(), Num=::NewEnumerator0)
   )
   &Drag(
    Drag=&MergeHelper_FloatSwitch(0=0.1, 1=0, 2=?, 3=?, Index=&SpawnGroup(), Num=::NewEnumerator0)
    Use Linear Drag=true
    Use Rotational Drag=true
   )
   &InheritSourceMovement(
    Applied Position Scale=&Multiply_VectorByFloat(
     Vector=<1 1 1>
     Float=&FloatFromCurve(FloatCurve=DataInterface() {propertyValues {Curve {keys [{arriveTangent=-3.2352, interpMode=RCIM_Cubic, leaveTangent=-3.2352, tangentMode=RCTM_User, time=0.05, value=1}, <at 1 0 RCIM_Cubic>]}}})
    )
    ApplyForce=false
    ApplyPosition=true
    ApplyPositionFalloff=false
   )
   &SolveForcesAndVelocity(Clamp Velocity=false, Limit Acceleration=false)
   &SubUVAnimation(
    End Frame Range Override=60
    Output Alternate Renderer Binding=true
    "Sprite Renderer "=DataInterface() {propertyValues {SpriteRenderer=&self:NE_VC_Boost_FireTrail_0.NiagaraSpriteRendererProperties_3}}
    Start Frame Range Override=&RandomRangeInt(Minimum=30, Maximum=35)
    SubUV Animation Mode:ENiagara_SubUVLookupModeV2=::NewEnumerator1
    SubUV Curve=DataInterface() {propertyValues {Curve {keys [
        {arriveTangent=-0.0074, interpMode=RCIM_Cubic, leaveTangent=-0.0074, tangentMode=RCTM_User, time=0, value=0}
        {arriveTangent=1.1179, interpMode=RCIM_Cubic, leaveTangent=1.1179, tangentMode=RCTM_User, time=0.2042, value=0.4}
        {arriveTangent=-1.1134, interpMode=RCIM_Cubic, leaveTangent=-1.1134, tangentMode=RCTM_User, time=1, value=2}
       ]}}}
    UseEndFrame=true
    UseStartFrame=true
   )
   &ScaleColor(
    Scale Alpha=&MergeHelper_FloatSwitch(
     0=&FloatFromCurve(
      FloatCurve=DataInterface() {propertyValues {Curve {keys [{interpMode=RCIM_Cubic, tangentMode=RCTM_User, time=0, value=0.2}, {interpMode=RCIM_Cubic, leaveTangent=1, time=0.8, value=0}]}}}
      Scale Curve=&MakeCustomFloatFromBool(
       Bool=$User.bIsPainted
       True Float=&FloatFromCurve(
        FloatCurve=DataInterface() {propertyValues {Curve {keys [
            {arriveTangent=31.8349, interpMode=RCIM_Cubic, leaveTangent=31.8349, tangentMode=RCTM_User, time=0, value=0.5}
            {arriveTangent=0.2212, interpMode=RCIM_Cubic, leaveTangent=0.2212, tangentMode=RCTM_User, time=0.1, value=1}
            {arriveTangent=-0.2087, interpMode=RCIM_Cubic, leaveTangent=-0.2087, tangentMode=RCTM_User, time=0.2, value=1}
            {arriveTangent=-15.7778, interpMode=RCIM_Cubic, leaveTangent=-15.7778, tangentMode=RCTM_User, time=0.4, value=0}
           ]}}}
        Scale Curve=&MakeCustomFloatFromBool(Bool=$User.bIsFrontendPreview, True Float=2, False Float=2)
       )
       False Float=&FloatFromCurve(
        FloatCurve=DataInterface() {propertyValues {Curve {keys [
            {arriveTangent=-0.7657, interpMode=RCIM_Cubic, time=0, value=0.8}
            {arriveTangent=-5.3343, interpMode=RCIM_Cubic, leaveTangent=-5.3343, tangentMode=RCTM_User, time=0.4, value=0}
           ]}}}
        Scale Curve=10
       )
      )
     )
     1=&FloatFromCurve(FloatCurve=DataInterface() {propertyValues {Curve {keys [<at 0.2 1 RCIM_Cubic>,<at 1.0011 0.5 RCIM_Cubic>]}}})
     2=?
     3=?
     Index=&SpawnGroup()
     Num=::NewEnumerator0
    )
    Scale RGB=&MergeHelper_VectorSwitch(
     0=&MakeCustomVectorFromBool(
      Bool=$User.bIsPainted
      True Vector=&VectorFromFloat(Value=&FloatFromCurve(
        FloatCurve=DataInterface() {propertyValues {Curve {keys [<at 0 1>,<at 1 0>]}}}
        Scale Curve=&MakeCustomFloatFromBool(Bool=$User.bIsFrontendPreview, True Float=5, False Float=1)
       ))
      False Vector=&VectorFromFloat(Value=&FloatFromCurve(FloatCurve=DataInterface() {propertyValues {Curve {keys [<at 0.39 1 RCIM_Cubic>, {arriveTangent=-10.9329, interpMode=RCIM_Cubic, leaveTangent=-10.9329, tangentMode=RCTM_User, time=0.4, value=1}]}}}))
     )
     1=&MakeCustomVectorFromBool(Bool=$User.bIsPainted, True Vector=<1 1 1>, False Vector=<5 5 5>)
     2=?
     3=?
     Index=&SpawnGroup()
     Num=::NewEnumerator0
    )
    ScaleA=true
    ScaleRGB=true
   ) {module=ScaleColor001}
  ]
  rendererClasses [&NiagaraSpriteRendererProperties]
  renderers [
   {rendererClass=&NiagaraSpriteRendererProperties}
   {rendererClass=&NiagaraSpriteRendererProperties, rendererIndex=1}
  ]
 }
 NE_VC_Boost_FireTrail001 {
  bEnabled=false
  emitterUpdateScript [
   &EmitterState(Life Cycle Mode=::NewEnumerator1, Loop Duration=1, UseLoopDelay=false)
   &VC_FilteredSocketCount() {enabled=false}
   &SpawnRate(
    SpawnRate=&MakeCustomFloatFromBool(
     Bool=$User.bIsFrontendPreview
     True Float=200
     False Float=&MakeCustomFloatFromBool(Bool=$User.bIsPainted, True Float=300, False Float=250)
    )
    Use Spawn Probability=false
   )
  ]
  particleSpawnScript [
   &VC_PaintedVariantsOverride(
    Black=<rgba 0.001 0.001 0.001 5>
    BurntSienna=<rgba 0.08 0.0213 0.016 0.7>
    Cobalt=&MakeCustomLinearColorFromBool(Bool=$User.bIsFrontendPreview, True Color=<rgba 0.0125 0 1.5 0.5>, False Color=<rgba 0 0 0.5 0.65>)
    Crimson=<rgba 0.1 0 0.0017 1>
    DefaultColor=<rgba 4 0.3333 0 0.2>
    ForestGreen=<rgba 0.03 0.3 0.03 0.4>
    Grey=<rgba 0.02 0.02 0.02 1>
    Lime=<rgba 0.2 0.4 0 0.4>
    Orange=<rgba 1 0.25 0 0.5>
    Pink=<rgba 0.35 0.042 0.238 1>
    Purple=<rgba 0.075 0 0.3 0.5>
    Saffron=<rgba 0.3 0.3 0 0.3>
    SkyBlue=<rgba 0 0.2994 0.45 0.5>
    TitaniumWhite=<rgba 0.5 0.5 0.5 0.2>
   )
   &InitializeParticle(
    Color=$Particles.Color
    Lifetime Max=1
    Lifetime Mode=::NewEnumerator1
    Mass Mode=::NewEnumerator3
    Position Offset=&AddVectorToPosition(
     Position=&AddVectorToPosition(Position=<-25 0 0>, Vector=$User.SingleBoostPos)
     Vector=&Multiply_VectorByFloat(Vector=$User.DTCarOffset, Float=1)
    )
    Position Offset Coordinate Space=::Local
    Sprite Rotation Mode=::NewEnumerator2
    Sprite Size Mode=::NewEnumerator3
    Uniform Sprite Size=&Multiply_Float(A=&RandomRangeFloat(Minimum=30, Maximum=55), B=$User.Scale)
    UsePositionOffset=true
    Write Material Random=true
   )
   &PartitionParticles(Number Of Partitions=$Emitter.PortCount) {enabled=false}
   &SkeletalMeshLocation(
    Bone / Socket Index=$Particles.PartitionParticles.Partition
    Bone / Socket Sampling Mode=::NewEnumerator5
    MaskSpawnGroup=false
    Mesh Sampling Type=::NewEnumerator6
    Sampled Position Offset=&AddVectorToPosition(Position=<-10 0 0>, Vector=$User.DTCarOffset)
    Skeletal Mesh=$System.SkeletalMesh
   ) {enabled=false}
   &AddVelocity(Velocity=&Multiply_VectorByFloat(Vector=<1000 0 0>, Float=$User.Scale), Velocity Speed Scale=-1)
   &ShapeLocation(
    MaskSpawnGroup=false
    Override Seed=false
    Sphere Radius=&RandomRangeFloat(Minimum=0, Maximum=25)
    UseOwnerScale=true
   ) {enabled=false}
  ]
  particleUpdateScript [
   &ParticleState()
   &SpriteRotationRate(Delta Time:NiagaraFloat=$Particles.Lifetime, Rotation Rate=&RandomRangeFloat(Minimum=10, Maximum=-10))
   &ScaleSpriteSize(
    Uniform Curve Scale=1.5
    Uniform Curve Sprite Scale=DataInterface() {propertyValues {Curve {keys [
        <at 0 0.5 RCIM_Cubic>
        {arriveTangent=-0.3122, interpMode=RCIM_Cubic, leaveTangent=-0.3122, tangentMode=RCTM_User, time=0.02, value=0.8}
        {arriveTangent=-0.9415, interpMode=RCIM_Cubic, leaveTangent=-0.9415, tangentMode=RCTM_User, time=0.5, value=0}
       ]}}}
   )
   &WindForce(Frequency=15, Scale=50, Wind Speed=<2 0 2>, Wind Speed Scale=100)
   &Drag(Drag=0.1, Use Linear Drag=true, Use Rotational Drag=true)
   &InheritSourceMovement(
    Applied Position Scale=&Multiply_VectorByFloat(
     Vector=<1 1 1>
     Float=&FloatFromCurve(FloatCurve=DataInterface() {propertyValues {Curve {keys [{arriveTangent=-3.2352, interpMode=RCIM_Cubic, leaveTangent=-3.2352, tangentMode=RCTM_User, time=0.05, value=1}, <at 1 0 RCIM_Cubic>]}}})
    )
    ApplyForce=false
    ApplyPosition=true
    ApplyPositionFalloff=false
   )
   &SolveForcesAndVelocity(Clamp Velocity=false, Limit Acceleration=false)
   &SubUVAnimation(
    End Frame Range Override=60
    "Sprite Renderer "=DataInterface() {propertyValues {SpriteRenderer=&self:NE_VC_Boost_FireTrail001_0.NiagaraSpriteRendererProperties_3}}
    Start Frame Range Override=&RandomRangeInt(Minimum=30, Maximum=35)
    SubUV Animation Mode:ENiagara_SubUVLookupModeV2=::NewEnumerator1
    SubUV Curve=DataInterface() {propertyValues {Curve {keys [
        {arriveTangent=-0.0074, interpMode=RCIM_Cubic, leaveTangent=-0.0074, tangentMode=RCTM_User, time=0, value=0}
        {arriveTangent=1.1179, interpMode=RCIM_Cubic, leaveTangent=1.1179, tangentMode=RCTM_User, time=0.2042, value=0.4}
        {arriveTangent=-1.1134, interpMode=RCIM_Cubic, leaveTangent=-1.1134, tangentMode=RCTM_User, time=1, value=2}
       ]}}}
    UseEndFrame=true
    UseStartFrame=true
   )
   &ScaleColor(
    Scale Alpha=&FloatFromCurve(
     FloatCurve=DataInterface() {propertyValues {Curve {keys [{interpMode=RCIM_Cubic, tangentMode=RCTM_User, time=0, value=0.2}, {interpMode=RCIM_Cubic, leaveTangent=1, time=0.8, value=0}]}}}
     Scale Curve=&MakeCustomFloatFromBool(
      Bool=$User.bIsPainted
      True Float=&FloatFromCurve(
       FloatCurve=DataInterface() {propertyValues {Curve {keys [
           {arriveTangent=31.8349, interpMode=RCIM_Cubic, leaveTangent=31.8349, tangentMode=RCTM_User, time=0, value=0.5}
           {arriveTangent=0.2212, interpMode=RCIM_Cubic, leaveTangent=0.2212, tangentMode=RCTM_User, time=0.1, value=1}
           {arriveTangent=-0.2087, interpMode=RCIM_Cubic, leaveTangent=-0.2087, tangentMode=RCTM_User, time=0.2, value=1}
           {arriveTangent=-15.7778, interpMode=RCIM_Cubic, leaveTangent=-15.7778, tangentMode=RCTM_User, time=0.4, value=0}
          ]}}}
       Scale Curve=&MakeCustomFloatFromBool(Bool=$User.bIsFrontendPreview, True Float=2, False Float=2)
      )
      False Float=&FloatFromCurve(
       FloatCurve=DataInterface() {propertyValues {Curve {keys [
           {arriveTangent=-0.7657, interpMode=RCIM_Cubic, time=0, value=0.8}
           {arriveTangent=-5.3343, interpMode=RCIM_Cubic, leaveTangent=-5.3343, tangentMode=RCTM_User, time=0.4, value=0}
          ]}}}
       Scale Curve=10
      )
     )
    )
    Scale RGB=&MakeCustomVectorFromBool(
     Bool=$User.bIsPainted
     True Vector=&VectorFromFloat(Value=&FloatFromCurve(
       FloatCurve=DataInterface() {propertyValues {Curve {keys [<at 0 1>,<at 1 0>]}}}
       Scale Curve=&MakeCustomFloatFromBool(Bool=$User.bIsFrontendPreview, True Float=10, False Float=10)
      ))
     False Vector=&VectorFromFloat(Value=&FloatFromCurve(
       FloatCurve=DataInterface() {propertyValues {Curve {keys [<at 0.39 1 RCIM_Cubic>, {arriveTangent=-10.9329, interpMode=RCIM_Cubic, leaveTangent=-10.9329, tangentMode=RCTM_User, time=0.4, value=0}]}}}
       Scale Curve=2.5
      ))
    )
    ScaleA=true
    ScaleRGB=true
   ) {module=ScaleColor001}
  ]
  rendererClasses [&NiagaraSpriteRendererProperties]
  renderers [{rendererClass=&NiagaraSpriteRendererProperties}]
 }
 NE_VC_Boost_Ribbon001 {
  emitterUpdateScript [
   &EmitterState(Life Cycle Mode=::NewEnumerator1, Loop Duration=1, UseLoopDelay=false)
   &VC_FilteredSocketCount()
   &VC_RateQualityHandler()
   &VC_SpawnRibbon_v1(SpawnRate=60)
  ]
  particleSpawnScript [
   &VC_PaintedVariantsOverride(
    Black=<rgba 0.08 0.08 0.08 1>
    BurntSienna=<rgba 0.2 0.0833 0.06 0.5>
    Cobalt=<rgba 0 0.013 0.7 0.5>
    Crimson=<rgba 0.3 0.015 0.0197 0.4>
    DefaultColor=<rgba 3 1.1 0 1.5>
    ForestGreen=<rgba 0.1575 0.3 0.015 0.5>
    Grey=<rgba 0.1 0.1 0.1 1>
    Lime=<rgba 0.155 0.2 0.02 0.4>
    Orange=<rgba 0.4 0.2467 0 1>
    Pink=<rgba 0.4 0.1 0.3 0.5>
    Purple=<rgba 0.18 0 0.4 0.4>
    Saffron=<rgba 0.45 0.45 0.225 0.15>
    SkyBlue=<rgba 0.05 0.425 0.5 0.5>
    TitaniumWhite=<rgba 0.3 0.3 0.3 0.4>
   )
   &VC_InitializeRibbon_v1(
    Color=$Particles.Color
    Custom Offset=true
    Lifetime=1
    Local Vector Offset=&AddVectorToPosition(
     Position=&AddVectorToPosition(Position=<-50 0 0>, Vector=$User.SingleBoostPos)
     Vector=&Multiply_VectorByFloat(Vector=$User.DTCarOffset, Float=1)
    )
    Offset Amount=1
    Ribbon Width:NiagaraFloat=&Multiply_Float(A=80, B=$User.Scale)
   )
   &AddVelocity(Velocity=<1000 0 0>, Velocity Speed Scale=&Multiply_Float(A=-1, B=$User.Scale)) {module=AddVelocity001}
   &ShapeLocation(MaskSpawnGroup=false, Override Seed=false, Sphere Radius=5, UseOwnerScale=true)
  ]
  particleUpdateScript [
   &ParticleState()
   &WindForce(Wind Speed Scale=2)
   &Drag(Drag=0.5, Use Linear Drag=true, Use Rotational Drag=true)
   &InheritSourceMovement(
    Applied Position Scale=&Multiply_VectorByFloat(
     Vector=<1 1 1>
     Float=&FloatFromCurve(FloatCurve=DataInterface() {propertyValues {Curve {keys [<at 0 1>,<at 1 0.05>]}}})
    )
    ApplyForce=false
    ApplyPosition=true
    ApplyPositionFalloff=false
   )
   &SolveForcesAndVelocity(Clamp Velocity=false, Limit Acceleration=false)
   &VC_UpdateRibbon_v1()
   &ScaleColor(
    Scale Alpha=&FloatFromCurve(
     FloatCurve=DataInterface() {propertyValues {Curve {keys [{arriveTangent=-3.893, interpMode=RCIM_Cubic, leaveTangent=-3.893, tangentMode=RCTM_User, time=0, value=1}, <at 0.8 0 RCIM_Cubic>]}}}
     Scale Curve=&MakeCustomFloatFromBool(Bool=$User.bIsPainted, True Float=9, False Float=5)
    )
    Scale RGB=&VectorFromFloat(Value=&MakeCustomFloatFromBool(Bool=$User.bIsPainted, True Float=20, False Float=3))
    ScaleA=true
    ScaleRGB=true
   )
   &ScaleRibbonWidth(Ribbon Width Scale=&FloatFromCurve(FloatCurve=DataInterface() {propertyValues {Curve {keys [<at 0 1>,<at 1 0>]}}}, Scale Curve=1.1))
  ]
  rendererClasses [&NiagaraRibbonRendererProperties]
  renderers [{rendererClass=&NiagaraRibbonRendererProperties}]
 }
 NE_VC_Boost_SparkGlow {
  bEnabled=false
  emitterUpdateScript [
   &EmitterState()
   &VC_FilteredSocketCount() {enabled=false}
   &SpawnRate(SpawnRate=50, Use Spawn Probability=false)
  ]
  particleSpawnScript [
   &VC_PaintedVariantsOverride(
    Black=<rgba 0.07 0.07 0.07 0.5>
    BurntSienna=<rgba 0.07 0.0148 0 1>
    Cobalt=<rgba 0.075 0.1225 1.5 1>
    Crimson=<rgba 0.7 0 0.0117 0.65>
    DefaultColor=<rgba 0.3 0.1046 0 1>
    ForestGreen=<rgba 0.16 0.8 0.1923 1>
    Grey=<rgba 0.05 0.05 0.05 1>
    Lime=<rgba 0.375 0.5 0.25 0.4>
    Orange=<rgba 0.6 0.2874 0.12 0.4>
    Pink=<rgba 0.5 0.25 0.4091 1>
    Purple=<rgba 0.0833 0 0.5 0.5>
    SkyBlue=<rgba 0 0.4658 0.7 1>
    TitaniumWhite=<rgba 1 1 1 1>
   )
   &InitializeParticle(
    Color=$Particles.Color
    Lifetime Max=0.1
    Lifetime Min:NiagaraFloat=&MakeCustomFloatFromBool(Bool=$User.bIsPainted, True Float=0.05, False Float=0.01)
    Lifetime Mode=::NewEnumerator1
    Position Offset=&AddVectorToPosition(
     Position=&AddVectorToPosition(Position=<-50 0 0>, Vector=$User.SingleBoostPos)
     Vector=&Multiply_VectorByFloat(Vector=$User.DTCarOffset, Float=1)
    )
    Position Offset Coordinate Space=::Local
    Sprite Size Mode=::NewEnumerator1
    Uniform Sprite Size Max=&Multiply_Float(A=&MakeCustomFloatFromBool(Bool=$User.bIsPainted, True Float=80, False Float=80), B=$User.Scale)
    Uniform Sprite Size Min=&Multiply_Float(A=25, B=$User.Scale)
    UsePositionOffset=true
    Write Material Random=true
   )
   &PartitionParticles(Number Of Partitions=$Emitter.PortCount) {enabled=false}
   &SkeletalMeshLocation(
    Bone / Socket Index=$Particles.PartitionParticles.Partition
    Bone / Socket Sampling Mode=::NewEnumerator5
    MaskSpawnGroup=false
    Mesh Sampling Type=::NewEnumerator6
    Sampled Position Offset=&AddVectorToPosition(Position=<-30 0 0>, Vector=$User.DTCarOffset)
    Skeletal Mesh=$System.SkeletalMesh
   ) {enabled=false}
   &AddVelocity(Velocity=<-1500 0 0>) {enabled=false}
  ]
  particleUpdateScript [
   &ParticleState()
   &ScaleSpriteSize(
    Uniform Curve Scale=0.38
    Uniform Curve Sprite Scale=DataInterface() {propertyValues {Curve {keys [{tangentMode=RCTM_User, time=0, value=1}, {tangentMode=RCTM_User, time=1, value=0}]}}}
   )
   &SolveForcesAndVelocity(Clamp Velocity=false, Limit Acceleration=false)
   &ScaleColor(
    Scale Alpha=&MakeCustomFloatFromBool(Bool=$User.bIsPainted, True Float=2, False Float=1)
    Scale RGB=&Multiply_VectorByFloat(Vector=<1 1 1>, Float=50)
    ScaleA=true
    ScaleRGB=true
   )
  ]
  rendererClasses [&NiagaraSpriteRendererProperties]
  renderers [{rendererClass=&NiagaraSpriteRendererProperties}]
 }
 NE_VC_Boost_SparkGlow001 {
  emitterUpdateScript [
   &EmitterState(Life Cycle Mode=::NewEnumerator1, Loop Duration=0.1, UseLoopDelay=false)
   &VC_FilteredSocketCount() {enabled=false}
   &SpawnRate(SpawnRate=50, Use Spawn Probability=false)
   &SpawnRate(SpawnGroup=1, SpawnRate=250, Use Spawn Probability=false) {module=SpawnRate001}
  ]
  particleSpawnScript [
   &VC_PaintedVariantsOverride(
    Black=<rgba 0.005 0.005 0.005 1>
    BurntSienna=<rgba 0.0573 0.0131 0.0012 0.5>
    Cobalt=<rgba 0.025 0.0408 0.5 0.5>
    Crimson=<rgba 0.3021 0 0.005 0.5>
    DefaultColor=<rgba 0.3 0.1046 0 1>
    ForestGreen=<rgba 0 0.2344 0.0118 0.5>
    Grey=<rgba 0.02 0.02 0.02 0.5>
    Lime=<rgba 0.0547 0.0729 0.0365 0.5>
    Orange=<rgba 0.5 0.2395 0.1 0.4>
    Pink=<rgba 0.0625 0.0098 0.0433 0.5>
    Purple=<rgba 0.0295 0 0.1771 0.5>
    R1 Black:LinearColor=<rgba 1 1 1 0.5>
    R1 BurntSienna:LinearColor=<rgba 5 2.7004 2.0833 0.5>
    R1 Cobalt:LinearColor=<rgba 3.75 3.9581 15 0.5>
    R1 Crimson:LinearColor=<rgba 15 2.3438 2.6868 1>
    R1 ForestGreen:LinearColor=<rgba 0 15 0.7579 1>
    R1 Grey:LinearColor=<rgba 3 3 3 1>
    R1 Lime:LinearColor=<rgba 9 18 0 1>
    R1 Orange:LinearColor=<rgba 5 1.744 0 1>
    R1 Pink:LinearColor=<rgba 15 1.8 10.2 1>
    R1 Purple:LinearColor=<rgba 12.3124 0 30 0.5>
    R1 Saffron:LinearColor=<rgba 20 15.3333 0 1>
    R1 SkyBlue:LinearColor=<rgba 0 9.9808 15 1>
    R1 TitaniumWhite:LinearColor=<rgba 15 15 15 1>
    Saffron=<rgba 0.05 0.0383 0 0.5>
    SkyBlue=<rgba 0 0.3327 0.5 0.5>
    TitaniumWhite=<rgba 0.5 0.5 0.5 0.5>
    Use R1 Painted Overrides=true
   )
   /SetVariables(Particles.VisibilityTag=$Emitter.SpawnGroup) {isSetParameters=true}
   &InitializeParticle(
    Color=$Particles.Color
    Lifetime=&MergeHelper_FloatSwitch(
     0=&RandomRangeFloat(
      Minimum=&MakeCustomFloatFromBool(Bool=$User.bIsPainted, True Float=0.05, False Float=0.01)
      Maximum=0.1
     )
     1=0.5
     2=?
     3=?
     Index=$Emitter.SpawnGroup
     Num=::NewEnumerator0
    )
    Position Offset=&MergeHelper_VectorSwitch(
     0=&AddVectorToPosition(
      Position=&AddVectorToPosition(Position=<-110 0 0>, Vector=$User.SingleBoostPos)
      Vector=&Multiply_VectorByFloat(Vector=$User.DTCarOffset, Float=1)
     )
     1=&AddVectorToPosition(
      Position=&AddVectorToPosition(Position=<-110 0 0>, Vector=$User.SingleBoostPos)
      Vector=&Multiply_VectorByFloat(Vector=$User.DTCarOffset, Float=1)
     )
     2=?
     3=?
     Index=$Emitter.SpawnGroup
     Num=::NewEnumerator0
    )
    Position Offset Coordinate Space=::Local
    Sprite Size Mode=::NewEnumerator3
    Uniform Sprite Size=&MergeHelper_FloatSwitch(
     0=&RandomRangeFloat(Minimum=200, Maximum=280)
     1=&RandomRangeFloat(Minimum=2, Maximum=4)
     2=?
     3=?
     Index=$Emitter.SpawnGroup
     Num=::NewEnumerator0
    )
    UsePositionOffset=true
    Write Material Random=true
   )
   &PartitionParticles(Number Of Partitions=$Emitter.PortCount) {enabled=false}
   &SkeletalMeshLocation(
    Bone / Socket Index=$Particles.PartitionParticles.Partition
    Bone / Socket Sampling Mode=::NewEnumerator5
    MaskSpawnGroup=false
    Mesh Sampling Type=::NewEnumerator6
    Sampled Position Offset=&AddVectorToPosition(Position=<-30 0 0>, Vector=$User.DTCarOffset)
    Skeletal Mesh=$System.SkeletalMesh
   ) {enabled=false}
   &AddVelocity(
    Cone Angle=360
    Cone Axis=<-1 0 1>
    Module.UseConeFalloff=false
    Module.UseVelDistribution=false
    Override Seed=false
    Velocity Mode=::NewEnumerator2
    Velocity Speed=&MergeHelper_FloatSwitch(0=0, 1=&Multiply_Float(A=$User.Scale, B=150), 2=?, 3=?, Index=$Emitter.SpawnGroup, Num=::NewEnumerator0)
   ) {module=AddVelocity001}
   &ShapeLocation(
    MaskSpawnGroup=false
    Override Seed=false
    Sphere Radius=&MergeHelper_FloatSwitch(0=0, 1=10, 2=?, 3=?, Index=$Emitter.SpawnGroup, Num=::NewEnumerator0)
    UseOwnerScale=true
   )
   &AddVelocity(
    Module.Use Falloff Distance=false
    Override Seed=false
    Velocity Mode=::NewEnumerator1
    Velocity Speed=&MergeHelper_FloatSwitch(0=0, 1=25, 2=?, 3=?, Index=$Emitter.SpawnGroup, Num=::NewEnumerator0)
   ) {module=AddVelocity002}
   &AddVelocity(Velocity=<-1500 0 0>) {enabled=false}
  ]
  particleUpdateScript [
   &ParticleState()
   &ScaleSpriteSize(
    Scale Sprite Size Mode=::NewEnumerator0
    Uniform Scale Factor=&MergeHelper_FloatSwitch(
     0=&FloatFromCurve(
      FloatCurve=DataInterface() {propertyValues {Curve {keys [{tangentMode=RCTM_User, time=0, value=1}, {tangentMode=RCTM_User, time=1, value=0}]}}}
      Scale Curve=0.38
     )
     1=&FloatFromCurve(
      FloatCurve=DataInterface() {propertyValues {Curve {keys [
          {tangentMode=RCTM_User, time=0, value=1}
          {arriveTangent=-0.0405, interpMode=RCIM_Cubic, leaveTangent=-0.0405, tangentMode=RCTM_User, time=0.2, value=1}
          {arriveTangent=0.0776, interpMode=RCIM_Cubic, leaveTangent=0.0776, tangentMode=RCTM_User, time=1, value=0}
         ]}}}
      Scale Curve=2.5
     )
     2=?
     3=?
     Index=&SpawnGroup()
     Num=::NewEnumerator0
    )
   )
   &SolveForcesAndVelocity(Clamp Velocity=false, Limit Acceleration=false)
   &ScaleColor(
    Scale Alpha=&MergeHelper_FloatSwitch(
     0=&MakeCustomFloatFromBool(Bool=$User.bIsPainted, True Float=2, False Float=1)
     1=&FloatFromCurve(FloatCurve=DataInterface() {propertyValues {Curve {keys [
          <at 0 0 RCIM_Cubic>
          {arriveTangent=0.002, interpMode=RCIM_Cubic, leaveTangent=0.002, time=0.3, value=1.875}
          {arriveTangent=-2.0605, interpMode=RCIM_Cubic, leaveTangent=-2.0605, tangentMode=RCTM_User, time=1.0011, value=0.002}
         ]}}})
     2=?
     3=?
     Index=&SpawnGroup()
     Num=::NewEnumerator0
    )
    Scale RGB=&MergeHelper_VectorSwitch(
     0=&Multiply_VectorByFloat(Vector=<1 1 1>, Float=50)
     1=&MakeCustomVectorFromBool(Bool=$User.bIsPainted, True Vector=<13 13 13>, False Vector=<5 5 5>)
     2=?
     3=?
     Index=&SpawnGroup()
     Num=::NewEnumerator0
    )
    ScaleA=true
    ScaleRGB=true
   )
  ]
  rendererClasses [&NiagaraSpriteRendererProperties]
  renderers [
   {rendererClass=&NiagaraSpriteRendererProperties}
   {rendererClass=&NiagaraSpriteRendererProperties, rendererIndex=1}
  ]
 }
 NE_VC_Boost_Sphere {
  emitterUpdateScript [
   &EmitterState(Inactive Response=::NewEnumerator1, Life Cycle Mode=::NewEnumerator1, Loop Duration=0.5, UseLoopDelay=false)
   &SpawnBurst_Instantaneous(Use Spawn Probability=false, UseLoopCountLimit=true)
   &VC_FilteredSocketCount() {enabled=false}
   &SpawnRate(SpawnRate=2, Use Spawn Probability=false)
  ]
  particleSpawnScript [
   &VC_PaintedVariantsOverride(
    Black=<rgba 0.001 0.001 0.001 1>
    BurntSienna=<rgba 0.3 0.088 0.06 2>
    Cobalt=&MakeCustomLinearColorFromBool(Bool=$User.bIsFrontendPreview, True Color=<rgba 0.1725 0.15 1.5 2>, False Color=<rgba 0.345 0.3 3 2>)
    Crimson=<rgba 0.2 0 0.0054 2>
    DefaultColor=<rgba 3 0.3 0.015 2>
    ForestGreen=<rgba 0.0417 0.5 0 1.5>
    Grey=<rgba 0.2 0.2 0.2 2>
    Lime=<rgba 0.5 1 0 1>
    Orange=<rgba 1.2 0.4186 0 1>
    Pink=<rgba 1 0.1198 0.6799 1>
    Purple=<rgba 0.3 0 1.2 2>
    Saffron=<rgba 0.8 0.632 0.08 1>
    SkyBlue=<rgba 0.3 0.7658 1 1>
    TitaniumWhite=<rgba 1 1 1 1>
   )
   &InitializeParticle(
    Color=$Particles.Color
    Mesh Scale Mode=::NewEnumerator3
    Mesh Uniform Scale=&Multiply_Float(A=&MakeCustomFloatFromBool(Bool=$User.bIsPainted, True Float=0.3, False Float=0.3), B=$User.Scale)
    Position Offset=&AddVectorToPosition(
     Position=&AddVectorToPosition(Position=<-105 0 0>, Vector=$User.SingleBoostPos)
     Vector=&Multiply_VectorByFloat(Vector=$User.DTCarOffset, Float=1)
    )
    Position Offset Coordinate Space=::Local
    Sprite Rotation Mode=::NewEnumerator2
    Sprite Size Mode=::NewEnumerator3
    UsePositionOffset=true
    Write Material Random=true
   )
   &PartitionParticles(Number Of Partitions=$Emitter.PortCount) {enabled=false}
   &SkeletalMeshLocation(
    Bone / Socket Index=$Particles.PartitionParticles.Partition
    Bone / Socket Sampling Mode=::NewEnumerator5
    MaskSpawnGroup=false
    Mesh Sampling Type=::NewEnumerator6
    Skeletal Mesh=$System.SkeletalMesh
   ) {enabled=false}
   &AddVelocity(Velocity=<-1500 0 0>) {enabled=false}
  ]
  particleUpdateScript [
   &ParticleState()
   &UpdateMeshOrientation(Rotation Rate=0.7, Rotation Vector=&RandomRangeVector(Minimum=<-1 -1 -1>, Maximum=<1 1 1>))
   &SolveForcesAndVelocity(Clamp Velocity=false, Limit Acceleration=false)
   &ScaleColor(
    Scale Alpha=0.75
    Scale RGB=&VectorFromFloat(Value=&MakeCustomFloatFromBool(Bool=$User.bIsPainted, True Float=15, False Float=3))
    ScaleA=true
    ScaleRGB=true
   )
  ]
  rendererClasses [&NiagaraMeshRendererProperties, &NiagaraSpriteRendererProperties]
  renderers [
   {rendererClass=&NiagaraMeshRendererProperties, rendererIndex=0}
   {rendererClass=&NiagaraSpriteRendererProperties, rendererIndex=1}
  ]
 }
 NE_VC_Boost_Stardust01 {
  emitterUpdateScript [
   &EmitterState(Life Cycle Mode=::NewEnumerator1, Loop Duration=0.1, UseLoopDelay=false)
   &VC_FilteredSocketCount() {enabled=false}
   &SpawnRate(SpawnRate=50, Use Spawn Probability=false)
  ]
  particleSpawnScript [
   &VC_PaintedVariantsOverride(
    Black=<rgba 0.05 0.05 0.05 0.3>
    BurntSienna=<rgba 0.4 0.1625 0.1625 0.9>
    Cobalt=<rgba 0.6562 0.9578 3 0.5>
    Crimson=<rgba 2 0.2421 0.2292 0.3>
    DefaultColor=<rgba 1.5 1.5 0.4 0.5>
    ForestGreen=<rgba 0.5039 1.5 0.3047 0.5>
    Grey=<rgba 0.04 0.04 0.04 0.3>
    Lime=<rgba 2 4 0 0.3>
    Orange=<rgba 0.6724 0.2346 0 0.3>
    Pink=<rgba 2 0.0938 1.3068 0.6>
    Purple=<rgba 0.5 0 2 0.3>
    Saffron=<rgba 0.9 0.69 0 0.3>
    SkyBlue=<rgba 0 3.3269 5 0.6>
    TitaniumWhite=<rgba 1 1 1 0.3>
   )
   &InitializeParticle(
    Color=$Particles.Color
    Lifetime Max=0.8
    Lifetime Min=0.5
    Lifetime Mode=::NewEnumerator1
    Position Offset=&AddVectorToPosition(
     Position=&AddVectorToPosition(Position=<-110 0 0>, Vector=$User.SingleBoostPos)
     Vector=&Multiply_VectorByFloat(Vector=$User.DTCarOffset, Float=1)
    )
    Position Offset Coordinate Space=::Local
    Sprite Rotation Mode=::NewEnumerator2
    Sprite Size Mode=::NewEnumerator1
    Uniform Sprite Size Max=13
    Uniform Sprite Size Min=7
    UsePositionOffset=true
    Write Material Random=true
   )
   &PartitionParticles(Number Of Partitions=$Emitter.PortCount) {enabled=false}
   &SkeletalMeshLocation(
    Bone / Socket Index=$Particles.PartitionParticles.Partition
    Bone / Socket Sampling Mode=::NewEnumerator5
    MaskSpawnGroup=false
    Mesh Sampling Type=::NewEnumerator6
    Sampled Position Offset=&AddVectorToPosition(Position=<-30 0 0>, Vector=$User.DTCarOffset)
    Skeletal Mesh=$System.SkeletalMesh
   ) {enabled=false}
   &AddVelocity(
    Cone Angle=360
    Cone Axis=<-1 0 1>
    Module.UseConeFalloff=false
    Module.UseVelDistribution=false
    Override Seed=false
    Velocity Mode=::NewEnumerator2
    Velocity Speed=&Multiply_Float(A=$User.Scale, B=150)
   ) {enabled=false}
   &ShapeLocation(MaskSpawnGroup=false, Override Seed=false, Sphere Radius=21, UseOwnerScale=true) {enabled=false}
   &AddVelocity(Module.Use Falloff Distance=false, Override Seed=false, Velocity Mode=::NewEnumerator1) {enabled=false, module=AddVelocity002}
  ]
  particleUpdateScript [
   &ParticleState()
   &ScaleSpriteSize(
    Uniform Curve Scale=12
    Uniform Curve Sprite Scale=DataInterface() {propertyValues {Curve {keys [
        {tangentMode=RCTM_User, time=0, value=0}
        {arriveTangent=0.0776, interpMode=RCIM_Cubic, leaveTangent=0.0776, tangentMode=RCTM_User, time=1, value=1.5}
       ]}}}
   )
   &SubUVAnimation(
    "Sprite Renderer "=DataInterface() {propertyValues {SpriteRenderer=&self:NE_VC_Boost_Stardust01_0.NiagaraSpriteRendererProperties_0}}
    UseEndFrame=false
    UseStartFrame=false
   )
   &SolveForcesAndVelocity(Clamp Velocity=false, Limit Acceleration=false)
   &ScaleColor(
    Scale Alpha=&FloatFromCurve(FloatCurve=DataInterface() {propertyValues {Curve {keys [
         <at 0 0>
         {arriveTangent=0.002, leaveTangent=0.002, time=0.5, value=0.4}
         {arriveTangent=-2.0605, leaveTangent=-2.0605, time=1.0011, value=0.002}
        ]}}})
    Scale RGB=<1.5 1.5 1.5>
    ScaleA=true
    ScaleRGB=true
   )
  ]
  rendererClasses [&NiagaraSpriteRendererProperties]
  renderers [{rendererClass=&NiagaraSpriteRendererProperties}]
 }
 NE_VC_Boost_Stardust01001 {
  emitterUpdateScript [
   &EmitterState(Life Cycle Mode=::NewEnumerator1, Loop Duration=0.1, UseLoopDelay=false)
   &VC_FilteredSocketCount() {enabled=false}
   &SpawnRate(SpawnRate=120, Use Spawn Probability=false)
  ]
  particleSpawnScript [
   &VC_PaintedVariantsOverride(
    Black=<rgba 0.001 0.001 0.001 0.9>
    BurntSienna=<rgba 0.06 0.0291 0.0244 2>
    Cobalt=<rgba 0.1458 0.1458 1 2>
    Crimson=<rgba 1 0.1823 0.2095 1>
    DefaultColor=<rgba 3 0.25 0.01 1>
    ForestGreen=<rgba 0.01 0.2 0.01 2>
    Grey=<rgba 0.5 0.5 0.5 1>
    Lime=<rgba 0.15 0.3 0 0.9>
    Pink=<rgba 1 0.1198 0.6799 1>
    Purple=<rgba 0.125 0 0.5 2>
   )
   &InitializeParticle(
    Color=$Particles.Color
    Lifetime Max=0.3
    Lifetime Min=0.1
    Lifetime Mode=::NewEnumerator1
    Position Offset=&AddVectorToPosition(
     Position=&AddVectorToPosition(Position=<-110 0 0>, Vector=$User.SingleBoostPos)
     Vector=&Multiply_VectorByFloat(Vector=$User.DTCarOffset, Float=1)
    )
    Position Offset Coordinate Space=::Local
    Sprite Rotation Mode=::NewEnumerator2
    Sprite Size Mode=::NewEnumerator1
    Uniform Sprite Size Max=13
    Uniform Sprite Size Min=7
    UsePositionOffset=true
    Write Material Random=true
   )
   &PartitionParticles(Number Of Partitions=$Emitter.PortCount) {enabled=false}
   &SkeletalMeshLocation(
    Bone / Socket Index=$Particles.PartitionParticles.Partition
    Bone / Socket Sampling Mode=::NewEnumerator5
    MaskSpawnGroup=false
    Mesh Sampling Type=::NewEnumerator6
    Sampled Position Offset=&AddVectorToPosition(Position=<-30 0 0>, Vector=$User.DTCarOffset)
    Skeletal Mesh=$System.SkeletalMesh
   ) {enabled=false}
   &AddVelocity(
    Cone Angle=360
    Cone Axis=<-1 0 1>
    Module.UseConeFalloff=false
    Module.UseVelDistribution=false
    Override Seed=false
    Velocity Mode=::NewEnumerator2
    Velocity Speed=&Multiply_Float(A=$User.Scale, B=150)
   ) {enabled=false}
   &ShapeLocation(MaskSpawnGroup=false, Override Seed=false, Sphere Radius=50, UseOwnerScale=true)
   &AddVelocity(Module.Use Falloff Distance=false, Override Seed=false, Velocity Mode=::NewEnumerator1) {enabled=false, module=AddVelocity002}
  ]
  particleUpdateScript [
   &ParticleState()
   &ScaleSpriteSize(
    Uniform Curve Scale=8
    Uniform Curve Sprite Scale=DataInterface() {propertyValues {Curve {keys [
        {tangentMode=RCTM_User, time=0, value=0}
        {arriveTangent=0.0776, interpMode=RCIM_Cubic, leaveTangent=0.0776, tangentMode=RCTM_User, time=1, value=1.5}
       ]}}}
   )
   &SubUVAnimation(
    "Sprite Renderer "=DataInterface() {propertyValues {SpriteRenderer=&self:NE_VC_Boost_Stardust01001_0.NiagaraSpriteRendererProperties_1}}
    UseEndFrame=false
    UseStartFrame=false
   )
   &SolveForcesAndVelocity(Clamp Velocity=false, Limit Acceleration=false)
   &ScaleColor(
    Scale Alpha=&FloatFromCurve(FloatCurve=DataInterface() {propertyValues {Curve {keys [
         <at 0 0>
         {arriveTangent=0.002, leaveTangent=0.002, time=0.5, value=0.4}
         {arriveTangent=-2.0605, leaveTangent=-2.0605, time=1.0011, value=0.002}
        ]}}})
    ScaleA=true
    ScaleRGB=true
   )
  ]
  rendererClasses [&NiagaraSpriteRendererProperties]
  renderers [{rendererClass=&NiagaraSpriteRendererProperties}]
 }
 NE_VC_Boost_Stardust01002 {
  emitterUpdateScript [
   &EmitterState(Life Cycle Mode=::NewEnumerator1, Loop Duration=0.1, UseLoopDelay=false)
   &VC_FilteredSocketCount() {enabled=false}
   &SpawnRate(SpawnRate=30, Use Spawn Probability=false)
  ]
  particleSpawnScript [
   &VC_PaintedVariantsOverride(
    Black=<rgba 0.001 0.001 0.001 0.9>
    BurntSienna=<rgba 0.151 0.0295 0.0231 0.9>
    Cobalt=<rgba 0.0486 0.0941 0.849 0.7>
    Crimson=<rgba 0.4 0.012 0.0249 0.9>
    DefaultColor=<rgba 4 0.3333 0 1>
    ForestGreen=<rgba 0.0733 0.2 0.01 0.9>
    Grey=<rgba 0.2 0.2 0.2 0.9>
    Lime=<rgba 0.15 0.3 0 0.9>
    Orange=<rgba 0.6724 0.2346 0 0.9>
    Pink=<rgba 0.5052 0 0.3215 0.9>
    Purple=<rgba 0.075 0 0.3 1>
    Saffron=<rgba 0.9 0.69 0 0.5>
    SkyBlue=<rgba 0 0.357 0.5365 0.9>
    TitaniumWhite=<rgba 0.5 0.5 0.5 0.9>
   )
   &InitializeParticle(
    Color=$Particles.Color
    Lifetime Max=1
    Lifetime Min=0.4
    Lifetime Mode=::NewEnumerator1
    Position Offset=&AddVectorToPosition(
     Position=&AddVectorToPosition(Position=<-110 0 0>, Vector=$User.SingleBoostPos)
     Vector=&Multiply_VectorByFloat(Vector=$User.DTCarOffset, Float=1)
    )
    Position Offset Coordinate Space=::Local
    Sprite Rotation Mode=::NewEnumerator2
    Sprite Size Mode=::NewEnumerator1
    Uniform Sprite Size Max=44
    Uniform Sprite Size Min=34
    UsePositionOffset=true
    Write Material Random=true
   )
   &PartitionParticles(Number Of Partitions=$Emitter.PortCount) {enabled=false}
   &SkeletalMeshLocation(
    Bone / Socket Index=$Particles.PartitionParticles.Partition
    Bone / Socket Sampling Mode=::NewEnumerator5
    MaskSpawnGroup=false
    Mesh Sampling Type=::NewEnumerator6
    Sampled Position Offset=&AddVectorToPosition(Position=<-30 0 0>, Vector=$User.DTCarOffset)
    Skeletal Mesh=$System.SkeletalMesh
   ) {enabled=false}
   &AddVelocity(
    Cone Angle=360
    Cone Axis=<-1 0 1>
    Module.UseConeFalloff=false
    Module.UseVelDistribution=false
    Override Seed=false
    Velocity Mode=::NewEnumerator2
    Velocity Speed=&Multiply_Float(A=$User.Scale, B=150)
   ) {enabled=false}
   &ShapeLocation(MaskSpawnGroup=false, Override Seed=false, Sphere Radius=3, UseOwnerScale=true)
   &AddVelocity(Module.Use Falloff Distance=false, Override Seed=false, Velocity Mode=::NewEnumerator1) {enabled=false, module=AddVelocity002}
  ]
  particleUpdateScript [
   &ParticleState()
   &ScaleSpriteSize(
    Uniform Curve Scale=3.5
    Uniform Curve Sprite Scale=DataInterface() {propertyValues {Curve {keys [
        {tangentMode=RCTM_User, time=0, value=0}
        {arriveTangent=0.0776, interpMode=RCIM_Cubic, leaveTangent=0.0776, tangentMode=RCTM_User, time=1, value=1.5}
       ]}}}
   )
   &SubUVAnimation(
    "Sprite Renderer "=DataInterface() {propertyValues {SpriteRenderer=&self:NE_VC_Boost_Stardust01002_0.NiagaraSpriteRendererProperties_1}}
    UseEndFrame=false
    UseStartFrame=false
   )
   &SolveForcesAndVelocity(Clamp Velocity=false, Limit Acceleration=false)
   &ScaleColor(
    Scale Alpha=&FloatFromCurve(FloatCurve=DataInterface() {propertyValues {Curve {keys [
         <at 0 0>
         {arriveTangent=0.002, leaveTangent=0.002, time=0.8, value=1}
         {arriveTangent=-2.0605, leaveTangent=-2.0605, time=1.0011, value=0.002}
        ]}}})
    Scale RGB=<2.5 2.5 2.5>
    ScaleA=true
    ScaleRGB=true
   )
  ]
  rendererClasses [&NiagaraSpriteRendererProperties]
  renderers [{rendererClass=&NiagaraSpriteRendererProperties}]
 }
 NE_VC_Boost_StardustTrail {
  bEnabled=false
  emitterUpdateScript [
   &EmitterState(Life Cycle Mode=::NewEnumerator1, Loop Duration=1, UseLoopDelay=false)
   &VC_FilteredSocketCount() {enabled=false}
   &SpawnRate(
    SpawnRate=&MakeCustomFloatFromBool(Bool=$User.bIsFrontendPreview, True Float=120, False Float=250)
    Use Spawn Probability=false
   )
  ]
  particleSpawnScript [
   &VC_PaintedVariantsOverride(
    Black=<rgba 0.005 0.005 0.005 10>
    BurntSienna=<rgba 0.15 0.0054 0.003 2>
    Crimson=<rgba 0.8 0.008 0.0344 1>
    DefaultColor=<rgba 1 0.3 0 1>
    ForestGreen=<rgba 0 0.5 0.0833 1>
    Grey=<rgba 0.03 0.03 0.03 2>
    Lime=<rgba 0.2333 0.4 0 2>
    Purple=<rgba 0.1 0 0.3 2>
   )
   &InitializeParticle(
    Color=$Particles.Color
    Lifetime Max=0.4
    Lifetime Min=0.3
    Lifetime Mode=::NewEnumerator1
    Position Offset=&AddVectorToPosition(
     Position=&AddVectorToPosition(Position=<-50 0 0>, Vector=$User.SingleBoostPos)
     Vector=&Multiply_VectorByFloat(Vector=$User.DTCarOffset, Float=1)
    )
    Position Offset Coordinate Space=::Local
    Sprite Rotation Mode=::NewEnumerator2
    Sprite Size Mode=::NewEnumerator1
    Uniform Sprite Size Max=&Multiply_Float(A=2.5, B=$User.Scale)
    Uniform Sprite Size Min=&RandomRangeFloat(Minimum=2.5, Maximum=$User.Scale)
    UsePositionOffset=true
    Write Material Random=true
   )
   &PartitionParticles(Number Of Partitions=$Emitter.PortCount) {enabled=false}
   &SkeletalMeshLocation(
    Bone / Socket Index=$Particles.PartitionParticles.Partition
    Bone / Socket Sampling Mode=::NewEnumerator5
    MaskSpawnGroup=false
    Mesh Sampling Type=::NewEnumerator6
    Sampled Position Offset=&AddVectorToPosition(Position=<-30 0 0>, Vector=$User.DTCarOffset)
    Skeletal Mesh=$System.SkeletalMesh
   ) {enabled=false}
   &AddVelocity(
    Velocity=&RandomRangeVector(Minimum=<1000 0 0>, Maximum=<1000 0 0>)
    Velocity Speed Scale=&Multiply_Float(
     A=&MakeCustomFloatFromBool(Bool=$User.bIsFrontendPreview, True Float=-1, False Float=-0.7)
     B=$User.Scale
    )
   )
   &ShapeLocation(
    MaskSpawnGroup=false
    Override Seed=false
    Sphere Radius=&Multiply_Float(A=25, B=$User.Scale)
    UseOwnerScale=true
   )
  ]
  particleUpdateScript [
   &ParticleState()
   &InheritSourceMovement(
    Applied Position Scale=&Multiply_VectorByFloat(
     Vector=<1 1 1>
     Float=&FloatFromCurve(FloatCurve=DataInterface() {propertyValues {Curve {keys [
          {arriveTangent=-3.8867, interpMode=RCIM_Cubic, leaveTangent=-3.8867, tangentMode=RCTM_User, time=0, value=1}
          {arriveTangent=-0.09, interpMode=RCIM_Cubic, leaveTangent=-0.09, tangentMode=RCTM_User, time=1, value=0}
         ]}}})
    )
    ApplyForce=false
    ApplyPosition=true
    ApplyPositionFalloff=false
   )
   &SolveForcesAndVelocity(Clamp Velocity=false, Limit Acceleration=false)
   &ScaleSpriteSize(
    Uniform Curve Scale=2
    Uniform Curve Sprite Scale=DataInterface() {propertyValues {Curve {keys [<at 0.2 1 RCIM_Cubic>,<at 1 0.5 RCIM_Cubic>]}}}
   )
   &ScaleColor(
    Scale Alpha=&FloatFromCurve(FloatCurve=DataInterface() {propertyValues {Curve {keys [<at 0.2 1 RCIM_Cubic>,<at 1.0011 0.5 RCIM_Cubic>]}}})
    Scale RGB=&MakeCustomVectorFromBool(Bool=$User.bIsPainted, True Vector=<20 20 20>, False Vector=<5 5 5>)
    ScaleA=true
    ScaleRGB=true
   )
  ]
  rendererClasses [&NiagaraSpriteRendererProperties]
  renderers [{rendererClass=&NiagaraSpriteRendererProperties}]
 }
}
summary {
 systemName=NS_Boost_Fireball
 emitters [
  {emitterName=NE_VC_Boost_FireTrail, rendererClasses [&NiagaraSpriteRendererProperties]}
  {emitterName=NE_VC_Boost_SparkGlow001, rendererClasses [&NiagaraSpriteRendererProperties]}
  {emitterName=NE_VC_Boost_Stardust01001, rendererClasses [&NiagaraSpriteRendererProperties]}
  {bEnabled=false, emitterName=NE_VC_Boost_StardustTrail, rendererClasses [&NiagaraSpriteRendererProperties]}
  {emitterName=NE_VC_Boost_Ribbon001, rendererClasses [&NiagaraRibbonRendererProperties]}
  {
   emitterName=NE_VC_Boost_Sphere
   rendererClasses [&NiagaraMeshRendererProperties, &NiagaraSpriteRendererProperties]
  }
  {emitterName=NE_VC_Boost_CoreBigFire, rendererClasses [&NiagaraSpriteRendererProperties]}
  {bEnabled=false, emitterName=NE_VC_Boost_FireTrail001, rendererClasses [&NiagaraSpriteRendererProperties]}
  {bEnabled=false, emitterName=NE_VC_Boost_SparkGlow, rendererClasses [&NiagaraSpriteRendererProperties]}
  {emitterName=NE_VC_Boost_Stardust01, rendererClasses [&NiagaraSpriteRendererProperties]}
  {emitterName=NE_VC_Boost_Stardust01002, rendererClasses [&NiagaraSpriteRendererProperties]}
 ]
 userVariables [
  User.bIsFrontendPreview {type=NiagaraBool, defaultValue=true}
  User.bIsPainted {type=NiagaraBool}
  User.ColorVariantID {type=NiagaraInt32}
  User.DTCarOffset {type=Vector3f, description="", defaultValue {struct=Vector3f, value=<0 0 0>}}
  User.PaintedColor {type=LinearColor, description="", defaultValue {struct=LinearColor, value=<rgba 1 1 1 1>}}
  User.PaintedScalar {type=NiagaraFloat, defaultValue=1, description=""}
  User.PortCount {type=NiagaraInt32, defaultValue=1}
  User.Scale {type=NiagaraFloat, defaultValue=2, description=""}
  User.SingleBoostPos {type=Vector3f, description="", defaultValue {struct=Vector3f, value=<0 0 0>}}
 ]
}
